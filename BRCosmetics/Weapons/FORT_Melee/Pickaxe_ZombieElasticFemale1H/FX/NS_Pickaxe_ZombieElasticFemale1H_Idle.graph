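# Idle Niagara effect for the ZombieElasticFemale1H pickaxe: a flame in user-set primary and secondary colors (purple and blue by default) with dark accents.
# Mesh flames with a light combine with glow, smoke, local sparks and movement-spawned world sparks.
# Two skeletal-position ribbons get opposite sine/cosine swirl for flame trails, while looping short-lived mesh bursts add color-split glitch flickers.
# Time-of-day scalars dim most glows outside night.
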NiagaraSystem /BRCosmetics/Weapons/FORT_Melee/Pickaxe_ZombieElasticFemale1H/FX/NS_Pickaxe_ZombieElasticFemale1H_Idle
dependencies {usedDataInterfaces [&NiagaraDataInterfaceCurve], usedDynamicInputs [], usedModules [&SystemState]}
emitters {
 Dark_Glow {
  emitterUpdateScript [
   &EmitterState()
   &SpawnRate(SpawnRate=3, Use Spawn Probability=false)
  ]
  particleSpawnScript [
   &InitializeParticle(Color=<rgba 0 0 0 2>, Sprite Size Mode=::NewEnumerator3, Uniform Sprite Size=50, UsePositionOffset=false, Write Material Random=true)
   &AddVelocity(Velocity=<-20 0 0>)
  ]
  particleUpdateScript [
   &SolveForcesAndVelocity(Clamp Velocity=false, Limit Acceleration=false)
   &ParticleState()
   &ScaleSpriteSize(Uniform Curve Sprite Scale=DataInterface() {propertyValues {Curve {keys [
        {interpMode=RCIM_Cubic, tangentMode=RCTM_User, time=0, value=1}
        {arriveTangent=-1.9003, interpMode=RCIM_Cubic, leaveTangent=-1.9003, tangentMode=RCTM_User, time=1, value=0}
       ]}}})
  ]
  rendererClasses [&NiagaraSpriteRendererProperties]
  renderers [{rendererClass=&NiagaraSpriteRendererProperties}]
 }
 Decal_Glow_Wrist {
  emitterUpdateScript [
   &EmitterState()
   &SpawnBurst_Instantaneous(Spawn Count=1, Use Spawn Probability=false)
  ]
  particleSpawnScript [
   &InitializeParticle(Color=$User.Primary Color, Sprite Size Mode=::NewEnumerator3, Uniform Sprite Size=25, UsePositionOffset=false, Write Material Random=true) {module=InitializeParticle001}
   &SystemLocation(MaskSpawnGroup=false, Offset=<10 0 0>)
  ]
  particleUpdateScript [
   &ParticleState()
   &DynamicMaterialParameters(
    Index 0 Param 1=1.5
    Module.Param0WriteEnabled=true
    Module.Param1WriteEnabled=false
    Module.Param2WriteEnabled=false
    Module.Param3WriteEnabled=false
   ) {module=DynamicMaterialParameters001}
   &ScaleColor(
    Scale RGB=&VectorFromFloat(Value=&Multiply_Float(A=2, B=&FNTimeOfDayScalar(Night Scale=1, Morning Scale=0.1, Evening Scale=0.3)))
    ScaleA=false
    ScaleRGB=true
   )
   &CameraOffset(Camera Offset Amount=20)
  ]
  rendererClasses [&NiagaraSpriteRendererProperties]
  renderers [{rendererClass=&NiagaraSpriteRendererProperties}]
 }
 Decal_Smoke_hand {
  emitterUpdateScript [
   &EmitterState()
   &SpawnRate(SpawnRate=3, Use Spawn Probability=false)
  ]
  particleSpawnScript [
   &InitializeParticle(
    Color=$User.Primary Color
    Lifetime Max=1.5
    Lifetime Mode=::NewEnumerator1
    Sprite Rotation Mode=::NewEnumerator2
    Sprite Size Mode=::NewEnumerator3
    Uniform Sprite Size=&UniformRangedFloat(Minimum=13, Maximum=20)
    UsePositionOffset=false
    Write Material Random=true
   ) {module=InitializeParticle001}
   &SystemLocation(MaskSpawnGroup=false, Offset=<20 0 0>)
   &AddVelocity(Velocity=&UniformRangedVector(Minimum=<-10 0 0>, Maximum=<-15 0 0>))
  ]
  particleUpdateScript [
   &ParticleState()
   &ScaleColor(
    Scale Alpha=&FloatFromCurve(
     FloatCurve=DataInterface() {propertyValues {Curve {keys [
         {interpMode=RCIM_Cubic, tangentMode=RCTM_Break, time=0, value=0}
         {interpMode=RCIM_Cubic, tangentMode=RCTM_User, time=0.4, value=0.9932}
         {interpMode=RCIM_Cubic, tangentMode=RCTM_Break, time=1, value=0}
        ]}}}
     Scale Curve=3
    )
    Scale RGB=&VectorFromFloat(Value=&Multiply_Float(A=3, B=&FNTimeOfDayScalar(Night Scale=1, Morning Scale=0.1, Evening Scale=0.3)))
    ScaleA=true
    ScaleRGB=true
   )
   &CameraOffset(Camera Offset Amount=20)
   &SolveForcesAndVelocity(Clamp Velocity=false, Limit Acceleration=false)
   &DynamicMaterialParameters(
    Index 0 Param 1=&FloatFromCurve(FloatCurve=DataInterface() {propertyValues {Curve {keys [{interpMode=RCIM_Cubic, leaveTangent=2.3044, tangentMode=RCTM_Break, time=0, value=0.25}, <at 1 1>]}}})
    Index 0 Param 2=0.5
    Module.Param0WriteEnabled=true
    Module.Param1WriteEnabled=true
    Module.Param2WriteEnabled=true
    Module.Param3WriteEnabled=true
   )
  ]
  rendererClasses [&NiagaraSpriteRendererProperties]
  renderers [{rendererClass=&NiagaraSpriteRendererProperties}]
 }
 FN_Contrails_Flame {
  emitterUpdateScript [
   &EmitterState()
   &FN_SpawnRibbon_v2(Number of Ribbons=2, Spawn Mode=::NewEnumerator0, SpawnRate=60)
  ]
  particleSpawnScript [
   &FN_InitializeRibbon_v3(Enable Skeletal Position=true, Lifetime=1, Offset Axis:ENiagaraOrientationAxis=::XAxis, Ribbon Width=20)
   &SystemLocation(MaskSpawnGroup=false, Offset=<13 0 0>)
   &AddVelocity(Velocity=&MakeVector(
     X=0
     Y=&MaskFloatBySpawnGroup(
      Masked Float=&Sine(
       Normalized Angle=&Multiply_Float(A=$Emitter.Age, B=-1)
       Period=0.6
       Scale=&FloatFromCurve(
        FloatCurve=DataInterface() {propertyValues {Curve {keys [<at 0 1 RCIM_Cubic>, {arriveTangent=-2.1595, interpMode=RCIM_Cubic, leaveTangent=-2.1595, tangentMode=RCTM_User, time=1, value=0}]}}}
        CurveIndex=&NormalizeDistanceRange(
         Start Position:NiagaraPosition=/Niagara/DynamicInputs/Helpers/SimulationPosition()
         End Position:NiagaraPosition=<0 0 0>
         Distance:NiagaraFloat=1.5
         Invert Normalized Range:NiagaraBool=false
        )
        Scale Curve=45
       )
       Bias=1
      )
      Passthrough Float=&Cosine(
       Normalized Angle=&Multiply_Float(A=$Emitter.Age, B=-1)
       Period=0.6
       Scale=&FloatFromCurve(
        FloatCurve=DataInterface() {propertyValues {Curve {keys [<at 0 1 RCIM_Cubic>, {arriveTangent=-2.1595, interpMode=RCIM_Cubic, leaveTangent=-2.1595, tangentMode=RCTM_User, time=1, value=0}]}}}
        CurveIndex=&NormalizeDistanceRange(
         Start Position:NiagaraPosition=&SimulationPosition()
         End Position:NiagaraPosition=<0 0 0>
         Distance:NiagaraFloat=1.5
         Invert Normalized Range:NiagaraBool=false
        )
        Scale Curve=45
       )
       Bias=1
      )
     )
     Z=&MaskFloatBySpawnGroup(
      Masked Float=&Cosine(
       Normalized Angle=&Multiply_Float(A=$Emitter.Age, B=-1)
       Period=0.6
       Scale=&FloatFromCurve(
        FloatCurve=DataInterface() {propertyValues {Curve {keys [<at 0 1 RCIM_Cubic>, {arriveTangent=-2.1595, interpMode=RCIM_Cubic, leaveTangent=-2.1595, tangentMode=RCTM_User, time=1, value=0}]}}}
        CurveIndex=&NormalizeDistanceRange(
         Start Position:NiagaraPosition=&SimulationPosition()
         End Position:NiagaraPosition=<0 0 0>
         Distance:NiagaraFloat=1.5
         Invert Normalized Range:NiagaraBool=false
        )
        Scale Curve=45
       )
       Bias=1
      )
      Passthrough Float=&Sine(
       Normalized Angle=&Multiply_Float(A=$Emitter.Age, B=-1)
       Period=0.6
       Scale=&FloatFromCurve(
        FloatCurve=DataInterface() {propertyValues {Curve {keys [<at 0 1 RCIM_Cubic>, {arriveTangent=-2.1595, interpMode=RCIM_Cubic, leaveTangent=-2.1595, tangentMode=RCTM_User, time=1, value=0}]}}}
        CurveIndex=&NormalizeDistanceRange(
         Start Position:NiagaraPosition=&SimulationPosition()
         End Position:NiagaraPosition=<0 0 0>
         Distance:NiagaraFloat=1.5
         Invert Normalized Range:NiagaraBool=false
        )
        Scale Curve=45
       )
       Bias=1
      )
     )
    ))
  ]
  particleUpdateScript [
   &ScaleRibbonWidth(Ribbon Width Scale=&FloatFromCurve(FloatCurve=DataInterface() {propertyValues {Curve {keys [<at 0 0.2 RCIM_Cubic>, {arriveTangent=-0.2, interpMode=RCIM_Cubic, leaveTangent=-0.2, time=0.1953, value=0.6379}, <at 1 0 RCIM_Cubic>]}}}))
   &Material(
    Base Texture Time:NiagaraFloat=&Multiply_Float(A=$Emitter.Age, B=&Lerp_Float(A=-0.4, B=-100, Alpha=$Particles.IntegratedDive))
    PreMult=0
   )
   /Niagara/Modules/Update/Forces/PointAttractionForce(
    Attraction Radius=40
    AttractionStrength=110
    AttractorPosition:Vector3f=&Add_Vector(A=&SimulationPosition(), B=<-15 0 0>)
    Falloff Exponent=0.18
    Module.Kill Within Radius=false
    Module.Use Falloff=true
   )
   &AddVelocity() {module=AddVelocity001}
   &SolveForcesAndVelocity(Clamp Velocity=false, Limit Acceleration=false)
   &FN_UpdateRibbon(
    Enable Skeletal Position=true
    Initial Color:LinearColor=$Particles.Initial.Color
    Initial RibbonWidth:NiagaraFloat=$Particles.Initial.RibbonWidth
   )
   /SetVariables(Particles.SpriteFacing=HlslExpression() {hlslExpression="dot(Particles.RibbonFacing,  float3(0,0,1))"}) {isSetParameters=true}
   &ScaleColor(
    Scale Alpha=&FloatFromCurve(FloatCurve=DataInterface() {propertyValues {Curve {keys [<at 0 1 RCIM_Cubic>,<at 1 0 RCIM_Cubic>]}}}, Scale Curve=0.5)
    Scale RGB=&Multiply_VectorByFloat(
     Vector=&Multiply_VectorByFloat(
      Vector=&Lerp_Vector(
       A=&MakeVectorFromLinearColorRGB(Color=$User.Primary Color)
       B=&MakeVectorFromLinearColorRGB(Color=$User.Secondary Color)
       Alpha=$Particles.NormalizedAge
      )
      Float=&FloatFromCurve(
       FloatCurve=DataInterface() {propertyValues {Curve {keys [
           <at 0 1 RCIM_Cubic>
           {arriveTangent=-1.7902, interpMode=RCIM_Cubic, leaveTangent=-1.7902, time=0.2294, value=0.5134}
           {arriveTangent=-0.6663, interpMode=RCIM_Cubic, leaveTangent=-0.6663, time=0.5107, value=0.0857}
           <at 1 0 RCIM_Cubic>
          ]}}}
       Scale Curve=10
      )
     )
     Float=&FNTimeOfDayScalar(Night Scale=1, Morning Scale=0.1, Evening Scale=0.3)
    )
    ScaleA=true
    ScaleRGB=true
   )
  ]
  rendererClasses [&NiagaraRibbonRendererProperties]
  renderers [{rendererClass=&NiagaraRibbonRendererProperties}]
 }
 Flame_MAIN {
  emitterUpdateScript [
   &EmitterState()
   &SpawnBurst_Instantaneous(Spawn Count=1, Use Spawn Probability=false)
  ]
  particleSpawnScript [
   &InitializeParticle(
    Color=$User.Primary Color
    Mesh Scale=&Multiply_Vector(A=<2.8 2.6 2.6>, B=&VectorFromFloat(Value=1))
    Mesh Scale Mode=::NewEnumerator4
    Sprite Size Mode=::NewEnumerator3
    Uniform Sprite Size=40
    UsePositionOffset=false
    Write Material Random=true
   ) {module=InitializeParticle001}
   /SetVariables(Particles.DynMatParm_2:Vector4f=&MakeVector4(X=1, Y=0, Z=0, W=0)) {isSetParameters=true, module=SetVariables_#1}
   /SetVariables(Particles.PosOffset_2:NiagaraPosition=$Particles.Position) {isSetParameters=true, module=SetVariables_#2}
   &SystemLocation(MaskSpawnGroup=false, Offset=<12 0 0>)
  ]
  particleUpdateScript [
   /SetVariables(Particles.LightColor=&MultiplyLinearColor(
     Linear Color=$User.Primary Color
     Scale Factor=&MakeVector4_FromFloat(Float=&Multiply_Float(A=20, B=1))
    )) {isSetParameters=true}
   /SetVariables(Particles.LightRadius=50) {isSetParameters=true}
   /SetVariables(Particles.LightExponent=3) {isSetParameters=true}
   /SetVariables(Particles.LightPosition:Vector3f=&Add_Vector(A=$Particles.PosOffset_2, B=<-10 5 0>)) {isSetParameters=true}
   &ParticleState()
   &UpdateMeshOrientation(Rotation Vector=<-1 0 0>) {module=UpdateMeshOrientation001}
   &self:DynParam_ColorPicker_01(Primary Color:LinearColor=$User.Primary Color, Secondary Color:LinearColor=$User.Secondary Color) {module=DynParam_ColorPicker_01}
   &ScaleColor(
    Scale Alpha=0.4
    Scale RGB=&VectorFromFloat(Value=&Multiply_Float(A=0.5, B=&FNTimeOfDayScalar(Night Scale=1, Morning Scale=0.1, Evening Scale=0.3)))
    ScaleA=true
    ScaleRGB=true
   )
  ]
  rendererClasses [
   &NiagaraMeshRendererProperties
   &NiagaraSpriteRendererProperties
   &NiagaraLightRendererProperties
  ]
  renderers [
   {rendererClass=&NiagaraMeshRendererProperties, rendererIndex=0}
   {rendererClass=&NiagaraSpriteRendererProperties, rendererIndex=1}
   {rendererClass=&NiagaraLightRendererProperties, rendererIndex=2}
   {rendererClass=&NiagaraMeshRendererProperties, rendererIndex=3}
  ]
 }
 Flame_outside_Dark {
  emitterUpdateScript [
   &EmitterState()
   &SpawnBurst_Instantaneous(Spawn Count=1, Use Spawn Probability=false)
  ]
  particleSpawnScript [
   &InitializeParticle(
    Color=<rgba 0 0 0 7>
    Mesh Scale=&Multiply_Vector(A=<3 3.3 3>, B=&VectorFromFloat(Value=1.2))
    Mesh Scale Mode=::NewEnumerator4
    Sprite Size=<40 40>
    Sprite Size Mode=::NewEnumerator4
    UsePositionOffset=false
    Write Material Random=true
   ) {module=InitializeParticle001}
   /SetVariables(Particles.DynMatParm_2:Vector4f=&MakeVector4(X=1, Y=0, Z=0, W=0)) {isSetParameters=true, module=SetVariables_#1}
   /SetVariables(Particles.PosOffset_2:NiagaraPosition=$Particles.Position) {isSetParameters=true, module=SetVariables_#2}
   &SystemLocation(MaskSpawnGroup=false, Offset=<12 0.5 0>)
  ]
  particleUpdateScript [
   &ParticleState()
   &UpdateMeshOrientation(Rotation Vector=<-1 0 0>) {module=UpdateMeshOrientation001}
   &self:DynParam_ColorPicker_01(Primary Color:LinearColor=<rgba 0 0 0 1>, Secondary Color:LinearColor=<rgba 0 0 0 1>) {module=DynParam_ColorPicker_01}
  ]
  rendererClasses [&NiagaraMeshRendererProperties]
  renderers [{rendererClass=&NiagaraMeshRendererProperties, rendererIndex=0}]
 }
 Glitch_diffusion {
  emitterUpdateScript [
   &EmitterState(Life Cycle Mode=::NewEnumerator1, Loop Duration=1.5, UseLoopDelay=true)
   &SpawnBurst_Instantaneous(Spawn Count=1, Use Spawn Probability=false)
  ]
  particleSpawnScript [
   &InitializeParticle(
    Color=&MakeLinearColorFromVectorAndFloat(
     "Vector (RGB)"=&MakeVectorFromLinearColorRGB(Color=&LerpLinearColors(StartColor=<rgba 0 0.8282 1.2 2>, EndColor=$User.Primary Color, LerpFactor=0.3))
     "Float (Alpha)"=2
    )
    Lifetime=0.3
    Mesh Scale=&Multiply_Vector(A=<2.7 2.7 2.7>, B=&VectorFromFloat(Value=1.2))
    Mesh Scale Mode=::NewEnumerator4
    UsePositionOffset=false
    Write Material Random=true
   ) {module=InitializeParticle001}
   &SystemLocation(MaskSpawnGroup=false, Offset=<12 0.5 0>)
  ]
  particleUpdateScript [
   /SetVariables(Particles.Mesh_Color02:LinearColor=&MakeLinearColorFromVectorAndFloat(
     "Vector (RGB)"=&Multiply_VectorByFloat(
      Vector=&MakeVectorFromLinearColorRGB(Color=&LerpLinearColors(StartColor=<rgba 0.0052 0 1 1>, EndColor=$User.Primary Color, LerpFactor=0.5))
      Float=&FNTimeOfDayScalar(Night Scale=1, Morning Scale=0.2, Evening Scale=0.3)
     )
     "Float (Alpha)"=2
    )) {isSetParameters=true}
   /SetVariables(Particles.Mesh_Color03:LinearColor=&MakeLinearColorFromVectorAndFloat(
     "Vector (RGB)"=&Multiply_VectorByFloat(
      Vector=&MakeVectorFromLinearColorRGB(Color=&LerpLinearColors(StartColor=<rgba 1 0 0 2>, EndColor=$User.Primary Color, LerpFactor=0.3))
      Float=&FNTimeOfDayScalar(Night Scale=1, Morning Scale=0.2, Evening Scale=0.3)
     )
     "Float (Alpha)"=2
    )) {isSetParameters=true}
   &ParticleState()
   &CameraOffset(Camera Offset Amount=15)
   &UpdateMeshOrientation(Rotation Vector=<-1 0 0>) {module=UpdateMeshOrientation001}
   &self:DynParam_ColorPicker_03() {module=DynParam_ColorPicker_03}
   &ScaleMeshSize(Scale Factor=&MakeVector(
     X=0.9
     Y=&FloatFromCurve(
      FloatCurve=DataInterface() {propertyValues {Curve {keys [
          {interpMode=RCIM_Cubic, tangentMode=RCTM_User, time=0, value=1}
          {interpMode=RCIM_Cubic, tangentMode=RCTM_User, time=0.2, value=1.3}
          {arriveTangent=-0.125, interpMode=RCIM_Cubic, leaveTangent=-0.125, time=0.4, value=0.05}
          {arriveTangent=1.5833, interpMode=RCIM_Cubic, leaveTangent=1.5833, time=0.6, value=1.4}
          {tangentMode=RCTM_User, time=1, value=1}
         ]}}}
      Scale Curve=&RandomRangeFloat(Minimum=0.8, Maximum=1.3, Evaluation Type=::NewEnumerator1)
     )
     Z=&FloatFromCurve(
      FloatCurve=DataInterface() {propertyValues {Curve {keys [
          {interpMode=RCIM_Cubic, tangentMode=RCTM_User, time=0, value=1}
          {interpMode=RCIM_Cubic, tangentMode=RCTM_User, time=0.3, value=1.3}
          {arriveTangent=-0.125, interpMode=RCIM_Cubic, leaveTangent=-0.125, time=0.5, value=0.15}
          {arriveTangent=-0.3, interpMode=RCIM_Cubic, leaveTangent=-0.3, time=0.7, value=1.4}
          {tangentMode=RCTM_User, time=1, value=1}
         ]}}}
      Scale Curve=&RandomRangeFloat(Minimum=0.8, Maximum=1.3, Evaluation Type=::NewEnumerator1)
     )
    ))
   &ScaleColor(
    Scale RGB=&Multiply_VectorByFloat(Vector=<1 1 1>, Float=&FNTimeOfDayScalar(Night Scale=1, Morning Scale=0.1, Evening Scale=0.3))
    ScaleA=true
    ScaleRGB=true
   )
  ]
  rendererClasses [&NiagaraMeshRendererProperties]
  renderers [
   {rendererClass=&NiagaraMeshRendererProperties, rendererIndex=0}
   {rendererClass=&NiagaraMeshRendererProperties, rendererIndex=1}
   {rendererClass=&NiagaraMeshRendererProperties, rendererIndex=2}
  ]
 }
 Glitch_main {
  emitterUpdateScript [
   &EmitterState(Life Cycle Mode=::NewEnumerator1, Loop Duration=1.5, UseLoopDelay=true)
   &SpawnBurst_Instantaneous(Spawn Count=1, Use Spawn Probability=false)
  ]
  particleSpawnScript [
   &InitializeParticle(
    Color=&LerpLinearColors(StartColor=<rgba 1 1 1 1>, EndColor=$User.Primary Color, LerpFactor=0.6)
    Lifetime=0.3
    Mesh Scale=&Multiply_Vector(A=<2.8 2.6 2.6>, B=&VectorFromFloat(Value=1))
    Mesh Scale Mode=::NewEnumerator4
    UsePositionOffset=false
    Write Material Random=true
   ) {module=InitializeParticle001}
   &SystemLocation(MaskSpawnGroup=false, Offset=<12 0 0>)
  ]
  particleUpdateScript [
   &ParticleState()
   &CameraOffset(Camera Offset Amount=15)
   &UpdateMeshOrientation(Rotation Vector=<-1 0 0>) {module=UpdateMeshOrientation001}
   &ScaleColor(
    Scale Alpha=5
    Scale RGB=&VectorFromFloat(Value=&Multiply_Float(A=0.5, B=&FNTimeOfDayScalar(Night Scale=1, Morning Scale=0.1, Evening Scale=0.3)))
    ScaleA=true
    ScaleRGB=true
   )
  ]
  rendererClasses [&NiagaraMeshRendererProperties]
  renderers [
   {rendererClass=&NiagaraMeshRendererProperties, rendererIndex=0}
   {rendererClass=&NiagaraMeshRendererProperties, rendererIndex=1}
  ]
 }
 Smoke_Dark {
  emitterUpdateScript [
   &EmitterState(Inactive Response=::NewEnumerator1, Life Cycle Mode=::NewEnumerator1, Loop Delay=1, Loop Duration=1, UseLoopDelay=false)
   &SpawnRate(SpawnRate=3, Use Spawn Probability=false)
  ]
  particleSpawnScript [
   &InitializeParticle(
    Color=<rgba 0 0 0 1>
    Lifetime Min=1.5
    Lifetime Mode=::NewEnumerator1
    Sprite Rotation Mode=::NewEnumerator2
    Sprite Size Mode=::NewEnumerator1
    Uniform Sprite Size Max=45
    Uniform Sprite Size Min=40
    UsePositionOffset=false
    Write Material Random=true
   ) {module=InitializeParticle001}
   &SubUVAnimation(End Frame=4, SubUV Animation Mode:ENiagara_SubUVLookupMode=::NewEnumerator2)
   &AddVelocityInCone(
    Cone Angle=10
    Cone Axis=<-1 0 0>
    Module.Use Velocity Falloff On Cone Axis=false
    Velocity Strength=&RandomRangeFloat(Minimum=30, Maximum=50)
   ) {module=AddVelocityInCone001}
   &SphereLocation(
    MaskSpawnGroup=false
    Offset=<10 0.5 0>
    Override Seed=false
    Sphere Radius=10
    Surface Only=false
    UseNonUniformScale=false
    UseOffset=true
    UseSphereOrientation=false
   )
  ]
  particleUpdateScript [
   &ParticleState()
   &Drag(Drag=0.5, Use Linear Drag=true, Use Rotational Drag=false)
   &CameraOffset(Camera Offset Amount=1)
   &SolveForcesAndVelocity(Clamp Velocity=false, Limit Acceleration=false)
   &SpriteRotationRate(Rotation Rate=&FloatFromCurve(
     FloatCurve=DataInterface() {propertyValues {Curve {keys [{interpMode=RCIM_Cubic, tangentMode=RCTM_Break, time=0, value=1}, {interpMode=RCIM_Cubic, tangentMode=RCTM_Break, time=1, value=0.489}]}}}
     Scale Curve=&Lerp_Float(A=-80, B=80, Alpha=$Particles.Initial.MaterialRandom)
    ))
   &ScaleSpriteSize(Scale Factor=&Multiply_Vector2(
     A=<1 1>
     B=&Vector2DFromFloat(Value=&FloatFromCurve(FloatCurve=DataInterface() {propertyValues {Curve {keys [
           {arriveTangent=-1.6051, interpMode=RCIM_Cubic, leaveTangent=-1.6051, tangentMode=RCTM_User, time=0, value=1.3}
           {interpMode=RCIM_Cubic, leaveTangent=-0.7604, tangentMode=RCTM_Break, time=0.7, value=0.8}
           <at 2 -0.14 RCIM_Cubic>
          ]}}}))
    ))
   &DynamicMaterialParameters(
    Index 0 Param 2=1
    Index 0 Param 3=1
    Index 0 Param 4=&FloatFromCurve(FloatCurve=DataInterface() {propertyValues {Curve {keys [<at 0 1>,<at 1 0.22>]}}})
    Module.Param0WriteEnabled=false
    Module.Param1WriteEnabled=true
    Module.Param2WriteEnabled=true
    Module.Param3WriteEnabled=true
   )
   &ScaleColor(
    Scale Alpha=&Multiply_Float(
     A=&Multiply_Float(
      A=&RandomRangeFloat(Minimum=4, Maximum=7)
      B=&FNTimeOfDayScalar(Night Scale=1, Morning Scale=1, Evening Scale=1)
     )
     B=&FloatFromCurve(
      FloatCurve=DataInterface() {propertyValues {Curve {keys [
          {interpMode=RCIM_Cubic, tangentMode=RCTM_User, time=0, value=0}
          {arriveTangent=2.125, interpMode=RCIM_Cubic, leaveTangent=2.125, time=0.2, value=0.7}
          {arriveTangent=-2.7763, interpMode=RCIM_Cubic, leaveTangent=-2.6672, tangentMode=RCTM_Break, time=0.4, value=0.85}
          {arriveTangent=-0.5522, interpMode=RCIM_Cubic, tangentMode=RCTM_Break, time=1, value=0}
         ]}}}
      Scale Curve=&FN_FrontEndFloatModifier(Gameplay=1, Front End=0.5)
     )
    )
    ScaleA=true
    ScaleRGB=true
   ) {module=ScaleColor001}
  ]
  rendererClasses [&NiagaraSpriteRendererProperties]
  renderers [{rendererClass=&NiagaraSpriteRendererProperties}]
 }
 Sparks_LocalSpace {
  emitterUpdateScript [
   &EmitterState(Enable Distance Culling=true, MaxDistance=2000, Module.bUseMaxDistance=true, Module.bUseMinDistance=false, Scalability Mode=::NewEnumerator1)
   &SpawnRate(SpawnRate=5, Use Spawn Probability=false)
  ]
  particleSpawnScript [
   &InitializeParticle(
    Color=$User.Primary Color
    Lifetime Max=1.5
    Lifetime Min=0.5
    Lifetime Mode=::NewEnumerator1
    Sprite Size=&UniformRangedVector2D(Minimum=<1.15 2.3>, Maximum=<2 3>)
    Sprite Size Mode=::NewEnumerator4
    UsePositionOffset=false
    Write Material Random=true
   ) {module=InitializeParticle001}
   &SystemLocation(MaskSpawnGroup=false)
   &CylinderLocation(
    Cylinder Height=6
    Cylinder Radius=6
    MaskSpawnGroup=false
    Offset=<12 0 0>
    Offset Coordinate Space:ENiagaraCoordinateSpace=::Local
    Offset Mode:ENiagara_OffsetMode=::NewEnumerator0
    Override Seed=false
    UseOwnerScale=true
   ) {module=CylinderLocation001}
   &AddVelocity(Velocity=&Multiply_VectorByFloat(Vector=&Multiply_Vector(A=$Particles.CylinderLocation001.CylinderNormal, B=<0 1 1>), Float=50)) {module=AddVelocity002}
   &AddVelocity(Velocity=&UniformRangedVector(Minimum=<-20 0 0>, Maximum=<-40 0 0>))
  ]
  particleUpdateScript [
   &ParticleState()
   /Niagara/Modules/Update/Velocity/VortexVelocity(Influence Falloff Radius=50, Module.Use Influence Falloff=false, Velocity Amount=-40, Vortex Axis:Vector3f=<1 0 0>)
   &PointAttractionForce(
    Attraction Radius=40
    AttractionStrength=300
    AttractorPosition:Vector3f=&Add_Vector(A=&SimulationPosition(), B=<-40 0 0>)
    Falloff Exponent=0.4
    Module.Kill Within Radius=false
    Module.Use Falloff=true
   )
   &SolveForcesAndVelocity(Clamp Velocity=false, Limit Acceleration=false)
   &ScaleColor(
    Scale Alpha=&FloatFromCurve(FloatCurve=DataInterface() {propertyValues {Curve {keys [
         {interpMode=RCIM_Cubic, tangentMode=RCTM_Break, time=0, value=0}
         <at 0.4 1>
         {interpMode=RCIM_Cubic, tangentMode=RCTM_User, time=0.65, value=1}
         {interpMode=RCIM_Cubic, tangentMode=RCTM_Break, time=1, value=0}
        ]}}})
    Scale RGB=<20 20 20>
    ScaleA=true
    ScaleRGB=true
   )
   &CameraOffset(Camera Offset Amount=10)
   &ScaleSpriteSize(Scale Factor=&Vector2DFromFloat(Value=&FloatFromCurve(FloatCurve=DataInterface() {propertyValues {Curve {keys [
          {arriveTangent=4.4129, interpMode=RCIM_Cubic, leaveTangent=4.4129, tangentMode=RCTM_User, time=0, value=0.5}
          <at 0.1967 1>
          <at 1 0 RCIM_Cubic>
         ]}}})))
  ]
  rendererClasses [&NiagaraSpriteRendererProperties]
  renderers [{rendererClass=&NiagaraSpriteRendererProperties}]
 }
 Sparks_WorldSpace {
  emitterUpdateScript [
   &EmitterState(Enable Distance Culling=true, MaxDistance=2000, Module.bUseMaxDistance=true, Module.bUseMinDistance=false, Scalability Mode=::NewEnumerator1)
   &SpawnPerUnit(
    Max Movement Threshold=2000
    Module.Use Max Movement Threshold=true
    Module.Use Movement Tolerance=false
    Spawn Per Unit=20
    Use Spawn Probability=false
   )
  ]
  particleSpawnScript [
   &InitializeParticle(
    Color=$User.Primary Color
    Lifetime Max=0.4
    Lifetime Min=0.15
    Lifetime Mode=::NewEnumerator1
    Sprite Rotation Mode=::NewEnumerator2
    Sprite Size=&UniformRangedVector2D(Minimum=<1 1>, Maximum=<2 2>)
    Sprite Size Mode=::NewEnumerator4
    UsePositionOffset=false
    Write Material Random=true
   ) {module=InitializeParticle001}
   &SystemLocation(MaskSpawnGroup=false, Offset=<-10 0 0>)
   &CylinderLocation(Cylinder Height=20, Cylinder Radius=6, MaskSpawnGroup=false, Override Seed=false, UseOwnerScale=true) {module=CylinderLocation001}
   &AddVelocity(Velocity=&Multiply_VectorByFloat(Vector=&Multiply_Vector(A=$Particles.CylinderLocation001.CylinderNormal, B=<0 1 1>), Float=50)) {module=AddVelocity002}
   &AddVelocity(Velocity=&UniformRangedVector(Minimum=<-20 0 0>, Maximum=<-40 0 0>))
  ]
  particleUpdateScript [
   &ParticleState()
   &AccelerationForce(Acceleration=<0 0 200>)
   /Niagara/Modules/Update/Forces/CurlNoiseForce(
    Noise Frequency=0.04
    Noise Strength=&FloatFromCurve(FloatCurve=DataInterface() {propertyValues {Curve {keys [<at 0 1>,<at 1 0>]}}}, Scale Curve=300)
    Pan Noise=true
    Pan Noise Field=<0 0 1>
   )
   &InheritVelocity(Module.Limit Speed=true, Velocity Scale=<0.05 0.05 0.05>)
   &SolveForcesAndVelocity(Clamp Velocity=false, Limit Acceleration=false)
   &ScaleColor(
    Scale Alpha=&FloatFromCurve(FloatCurve=DataInterface() {propertyValues {Curve {keys [
         {interpMode=RCIM_Cubic, tangentMode=RCTM_Break, time=0, value=0}
         <at 0.4 1>
         {interpMode=RCIM_Cubic, tangentMode=RCTM_User, time=0.65, value=1}
         {interpMode=RCIM_Cubic, tangentMode=RCTM_Break, time=1, value=0}
        ]}}})
    Scale RGB=<20 20 20>
    ScaleA=true
    ScaleRGB=true
   )
   &CameraOffset(Camera Offset Amount=10)
   &ScaleSpriteSize(Scale Factor=&Vector2DFromFloat(Value=&FloatFromCurve(FloatCurve=DataInterface() {propertyValues {Curve {keys [
          {arriveTangent=4.4129, interpMode=RCIM_Cubic, leaveTangent=4.4129, tangentMode=RCTM_User, time=0, value=0.5}
          <at 0.1967 1>
          <at 1 0 RCIM_Cubic>
         ]}}})))
  ]
  rendererClasses [&NiagaraSpriteRendererProperties]
  renderers [{rendererClass=&NiagaraSpriteRendererProperties}]
 }
}
summary {
 systemName=NS_Pickaxe_ZombieElasticFemale1H_Idle
 emitters [
  {
   emitterName=Flame_MAIN
   rendererClasses [
    &NiagaraMeshRendererProperties
    &NiagaraSpriteRendererProperties
    &NiagaraLightRendererProperties
   ]
  }
  {emitterName=Decal_Glow_Wrist, rendererClasses [&NiagaraSpriteRendererProperties]}
  {emitterName=Sparks_LocalSpace, rendererClasses [&NiagaraSpriteRendererProperties]}
  {emitterName=Sparks_WorldSpace, rendererClasses [&NiagaraSpriteRendererProperties]}
  {emitterName=FN_Contrails_Flame, rendererClasses [&NiagaraRibbonRendererProperties]}
  {emitterName=Flame_outside_Dark, rendererClasses [&NiagaraMeshRendererProperties]}
  {emitterName=Smoke_Dark, rendererClasses [&NiagaraSpriteRendererProperties]}
  {emitterName=Glitch_main, rendererClasses [&NiagaraMeshRendererProperties]}
  {emitterName=Glitch_diffusion, rendererClasses [&NiagaraMeshRendererProperties]}
  {emitterName=Decal_Smoke_hand, rendererClasses [&NiagaraSpriteRendererProperties]}
  {emitterName=Dark_Glow, rendererClasses [&NiagaraSpriteRendererProperties]}
 ]
 userVariables [
  User.bIsFrontEnd {type=NiagaraBool}
  User.bIsFrontEndPreview {type=NiagaraBool}
  User.Primary Color {type=LinearColor, description="", defaultValue {struct=LinearColor, value=<rgba 0.7 0.05 0.9 1>}}
  User.RibbonIDOffset {type=NiagaraInt32}
  User.Secondary Color {type=LinearColor, description="", defaultValue {struct=LinearColor, value=<rgba 0.08 0.03 0.91 1>}}
  User.SkeletalMesh {
   type=NiagaraDataInterfaceSkeletalMesh
   defaultValue {
    struct=NiagaraExt_VariableValue_DataInterface
    value {dataInterface=&self:NiagaraDataInterfaceSkeletalMesh_1}
   }
  }
 ]
}
